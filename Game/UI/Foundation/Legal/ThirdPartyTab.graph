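# Activatable tab widget on Fortnite's legal screen that displays the third-party software notices and license texts.
# A ScrollBox with gradient top and bottom shadows wraps one auto-wrapping rich text block holding the hardcoded notice, styled by LegalScreenRichTextStyles with headers and hyperlinks.
# On activation, the SetupScrollSTW event passes the ScrollBox to SetScrollWidget.
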WidgetBlueprint /Game/UI/Foundation/Legal/ThirdPartyTab
graphs [EventGraph {nodes [
   SetupScrollSTW() {is=title K2Node_CustomEvent}
   > SetScrollWidget!(execute, ToScroll:Widget Object Reference=Get ScrollBox_2() -> (ScrollBox_2:Scroll Box Object Reference)) -> (then)
   Event On Activated()
   > SetupScrollSTW() {is=title K2Node_CallFunction}
  ]}]
variables []
widgets {
 info {inheritedWidgetCount=0, parentClass=&CommonActivatableWidget, rootWidgetClass=&ScrollBox, widgetCount=2}
 widgets [
  ScrollBox_2 {
   widgetClassPath=&ScrollBox
   bIsVariable=true
   widget=ScrollBox_2
   properties {widgetStyle {
     barThickness=9
     bottomShadowBrush {
      bIsDynamicallyLoaded=false
      drawAs=Image
      imageType=FullColor
      mirroring=NoMirror
      resourceName=../../../Engine/Content/Slate/Common/ScrollBoxShadowBottom.png
      resourceObject=/FortUILibrary/MaterialInstances/MI_UI_LinearGradientBottom
      tiling=Horizontal
      imageSize=<64 64>
      margin=<ltrb 0.5 0 0.5 1>
      outlineSettings {
       bUseBrushTransparency=false
       roundingType=HalfHeightRadius
       width=0
       color {colorUseRule=UseColor_Specified, specifiedColor=<rgba 0 0 0 0>}
       cornerRadii=<0 0 0 0>
      }
      tintColor {colorUseRule=UseColor_Specified, specifiedColor=<rgba 0.0024 0.016 0.117 1>}
      uVRegion {bIsValid=false, max=<0 0>, min=<0 0>}
     }
     horizontalScrolledContentPadding=<ltrb 0 0 1 0>
     leftShadowBrush {
      bIsDynamicallyLoaded=false
      drawAs=NoDrawType
      imageType=FullColor
      mirroring=NoMirror
      resourceName=../../../Engine/Content/Slate/Common/ScrollBoxShadowLeft.png
      resourceObject=null
      tiling=NoTile
      imageSize=<8 16>
      margin=<ltrb 1 0.5 0 0.5>
      outlineSettings {
       bUseBrushTransparency=false
       roundingType=HalfHeightRadius
       width=0
       color {colorUseRule=UseColor_Specified, specifiedColor=<rgba 0 0 0 0>}
       cornerRadii=<0 0 0 0>
      }
      tintColor {colorUseRule=UseColor_Specified, specifiedColor=<rgba 1 1 1 1>}
      uVRegion {bIsValid=false, max=<0 0>, min=<0 0>}
     }
     rightShadowBrush {
      bIsDynamicallyLoaded=false
      drawAs=NoDrawType
      imageType=FullColor
      mirroring=NoMirror
      resourceName=../../../Engine/Content/Slate/Common/ScrollBoxShadowRight.png
      resourceObject=null
      tiling=NoTile
      imageSize=<8 16>
      margin=<ltrb 0 0.5 1 0.5>
      outlineSettings {
       bUseBrushTransparency=false
       roundingType=HalfHeightRadius
       width=0
       color {colorUseRule=UseColor_Specified, specifiedColor=<rgba 0 0 0 0>}
       cornerRadii=<0 0 0 0>
      }
      tintColor {colorUseRule=UseColor_Specified, specifiedColor=<rgba 1 1 1 1>}
      uVRegion {bIsValid=false, max=<0 0>, min=<0 0>}
     }
     topShadowBrush {
      bIsDynamicallyLoaded=false
      drawAs=Image
      imageType=FullColor
      mirroring=NoMirror
      resourceName=../../../Engine/Content/Slate/Common/ScrollBoxShadowTop.png
      resourceObject=/FortUILibrary/MaterialInstances/MI_UI_LinearGradientTop
      tiling=Horizontal
      imageSize=<64 64>
      margin=<ltrb 0.5 1 0.5 0>
      outlineSettings {
       bUseBrushTransparency=false
       roundingType=HalfHeightRadius
       width=0
       color {colorUseRule=UseColor_Specified, specifiedColor=<rgba 0 0 0 0>}
       cornerRadii=<0 0 0 0>
      }
      tintColor {colorUseRule=UseColor_Specified, specifiedColor=<rgba 0.0024 0.016 0.117 1>}
      uVRegion {bIsValid=false, max=<0 0>, min=<0 0>}
     }
     verticalScrolledContentPadding=<ltrb 0 0 0 1>
    }}
  }
  FortRichTextBlock_0 {
   widgetClassPath=/Script/FortniteUI.FortRichTextBlockLegacy
   bIsVariable=true
   parent=ScrollBox_2
   widget=FortRichTextBlock_0
   properties {
    allowContextMenu=false
    autoWrapText=true
    bIsVolatile=true
    disableTouchInput=true
    styleSet=&/LegalScreenRichTextStyles
    text="<Header>Fortnite® Third Party Software Notice</>\r\n\r\n<Subheader>I. (ACES) Academy Color Encoding System</>\r\n\r\nLicense Terms for Academy Color Encoding System Components\r\n\r\nAcademy Color Encoding System (ACES) software and tools are provided by the Academy under the following terms and conditions: A worldwide, royalty-free, non-exclusive right to copy, modify, create derivatives, and use, in source and binary forms, is hereby granted, subject to acceptance of this license.\r\n\r\nCopyright (c) 2013 Academy of Motion Picture Arts and Sciences (A.M.P.A.S.). Portions contributed by others as indicated. All rights reserved.\r\n\r\nPerformance of any of the aforementioned acts indicates acceptance to be bound by the following terms and conditions:\r\n\r\nCopies of source code, in whole or in part, must retain the above copyright notice, this list of conditions and the Disclaimer of Warranty.\r\n\r\nUse in binary form must retain the above copyright notice, this list of conditions and the Disclaimer of Warranty in the documentation and/or other materials provided with the distribution.\r\n\r\nNothing in this license shall be deemed to grant any rights to trademarks, copyrights, patents, trade secrets or any other intellectual property of A.M.P.A.S. or any contributors, except as expressly stated herein.\r\n\r\nNeither the name \"A.M.P.A.S.\" nor the name of any other contributors to this software may be used to endorse or promote products derivative of or based on this software without express prior written permission of A.M.P.A.S. or the contributors, as appropriate.\r\n\r\nThis license shall be construed pursuant to the laws of the State of California, and any disputes related thereto shall be subject to the jurisdiction of the courts therein.\r\n\r\nDisclaimer of Warranty: THIS SOFTWARE IS PROVIDED BY A.M.P.A.S. AND CONTRIBUTORS \"AS IS\" AND ANY EXPRESS OR IMPLIED WARRANTIES, INCLUDING, BUT NOT LIMITED TO, THE IMPLIED WARRANTIES OF MERCHANTABILITY, FITNESS FOR A PARTICULAR PURPOSE, AND NON-INFRINGEMENT ARE DISCLAIMED. IN NO EVENT SHALL A.M.P.A.S., OR ANY CONTRIBUTORS OR DISTRIBUTORS, BE LIABLE FOR ANY DIRECT, INDIRECT, INCIDENTAL, SPECIAL, EXEMPLARY, RESITUTIONARY, OR CONSEQUENTIAL DAMAGES (INCLUDING, BUT NOT LIMITED TO, PROCUREMENT OF SUBSTITUTE GOODS OR SERVICES; LOSS OF USE, DATA, OR PROFITS; OR BUSINESS INTERRUPTION) HOWEVER CAUSED AND ON ANY THEORY OF LIABILITY, WHETHER IN CONTRACT, STRICT LIABILITY, OR TORT (INCLUDING NEGLIGENCE OR OTHERWISE) ARISING IN ANY WAY OUT OF THE USE OF THIS SOFTWARE, EVEN IF ADVISED OF THE POSSIBILITY OF SUCH DAMAGE.\r\n\r\nWITHOUT LIMITING THE GENERALITY OF THE FOREGOING, THE ACADEMY SPECIFICALLY DISCLAIMS ANY REPRESENTATIONS OR WARRANTIES WHATSOEVER RELATED TO PATENT OR OTHER INTELLECTUAL PROPERTY RIGHTS IN THE ACADEMY COLOR ENCODING SYSTEM, OR APPLICATIONS THEREOF, HELD BY PARTIES OTHER THAN A.M.P.A.S.,WHETHER DISCLOSED OR UNDISCLOSED.\r\n\r\n<Subheader>II. Android Open Source Project</>\r\n\r\nAndroid Open Source Project software is licensed under the Apache License Version 2.0.\r\n\r\n<Subheader>III. BattlEye</>\r\n\r\nThis License Agreement (“Agreement”) is made between you (the “Licensee”) and BattlEye Innovations e.K. (the “Licensor”). The terms of this Agreement apply to all current and future versions and updates of BattlEye anti-cheat software (“BattlEye”).\r\n\r\nBy installing, enabling or using BattlEye, Licensee agrees with all the terms of this Agreement. Licensor reserves all rights not specifically granted and transferred to Licensee.\r\n\r\nLicensee understands, acknowledges and agrees with the following:\r\n- Licensor grants Licensee a non-exclusive and non-transferable license to use BattlEye for non-commercial purposes only. Licensee therefore does not own BattlEye, Licensor remains the owner of BattlEye.\r\n\r\n- Licensor provides BattlEye on an “as is” basis without warranty of any kind. Licensor neither guarantees the correct, error-free functioning of BattlEye nor is Licensor responsible for any damage caused by the use of BattlEye.\r\n\r\n- Licensee may not decompile, disassemble, reverse-engineer, modify or redistribute BattlEye in any way.\r\n\r\n- BattlEye will automatically, without notice to Licensee, download and install updates from time to time.\r\n\r\n- BattlEye may scan Licensee’s entire random access memory (RAM), and any game-related and system-related files and folders on Licensee’s system using cheat-program- identifying algorithms, report results of such algorithms to other connected computers and/or to Licensor and process such information for the sole purpose of preventing and detecting the use of cheat programs. BattlEye only scans and/or processes data which is reasonably necessary to be scanned and/or processed to meet this purpose.\r\n\r\n- BattlEye may collect and process Licensee’s Internet Protocol address, game account name and identifier, in-game nickname, and system-related and hardware-related information including, but not limited to, device identifiers and hardware serial numbers.\r\n\r\n- Licensor may share the information with Epic Games, Inc. and its affiliates (collectively, “Epic”) and/or Licensor’s partners and/or affiliates.\r\n\r\n- If you are a resident of the European Union, Licensor may transfer your personal information for the above purposes to its servers in the United States or to Epic or to Licensor’s subsidiaries, affiliates, and service providers that may be located outside of the European Union. The laws in those jurisdictions may not necessarily provide the same level of data protection compared to the laws in your country.\r\n\r\n- Licensee acknowledges that the invasive nature of BattlEye is necessary to meet its purpose and goal of preventing and detecting cheat programs.\r\n\r\n- Licensor is allowed to terminate the license at any time for any reason and without notice to Licensee.\r\n\r\nThis License Agreement constitutes the entire agreement between Licensor and Licensee and supersedes any prior statements.\r\n\r\n<Subheader>IV. Burbank Font</>\r\n\r\nFortnite uses the Burbank font, the trademarks and copyright for which is owned by House Industries.\r\n\r\n<Subheader>V. Chromium Embedded Framework 3.2623.1395</>\r\n\r\nCopyright (c) 2008-2014 Marshall A. Greenblatt. Portions Copyright (c)\r\n// 2006-2009 Google Inc. All rights reserved.\r\n\r\nChromium Embedded Framework software is licensed under the New BSD License.\r\n\r\n<Subheader>VI. Cubemapgen</>\r\n\r\nCopyright (c) 2013, AMD, Inc. All rights reserved.\r\n\r\nCubemapgen software is licensed under the New BSD License.\r\n\r\n<Subheader>VII. Expat XML Parser</>\r\n\r\nCopyright (c) 1998-2000 Thai Open Source Software Center Ltd and Clark Cooper\r\nCopyright (c) 2001-2016 Expat maintainers\r\n\r\nPermission is hereby granted, free of charge, to any person obtaining a copy of this software and associated documentation files (the \"Software\"), to deal in the Software without restriction, including without limitation the rights to use, copy, modify, merge, publish, distribute, sublicense, and/or sell copies of the Software, and to permit persons to whom the Software is furnished to do so, subject to the following conditions:\r\n\r\nThe above copyright notice and this permission notice shall be included in all copies or substantial portions of the Software.\r\n\r\nTHE SOFTWARE IS PROVIDED \"AS IS\", WITHOUT WARRANTY OF ANY KIND, EXPRESS OR IMPLIED, INCLUDING BUT NOT LIMITED TO THE WARRANTIES OF MERCHANTABILITY, FITNESS FOR A PARTICULAR PURPOSE AND NONINFRINGEMENT. IN NO EVENT SHALL THE AUTHORS OR COPYRIGHT HOLDERS BE LIABLE FOR ANY CLAIM, DAMAGES OR OTHER LIABILITY, WHETHER IN AN ACTION OF CONTRACT, TORT OR OTHERWISE, ARISING FROM, OUT OF OR IN CONNECTION WITH THE SOFTWARE OR THE USE OR OTHER DEALINGS IN THE SOFTWARE.\r\n\r\n<Subheader>VIII. Font Awesome 4.5</>\r\n\r\nFont Awesome 4.5 is licensed under the SIL Open Font License, Version 1.1.\r\n\r\n<Subheader>IX. Freetype 2 version 2.6</>\r\n\r\nCopyright 1996-2002, 2006 by David Turner, Robert Wilhelm, and Werner Lemberg\r\n\r\nFortnite is based in part of the work of the FreeType Team.\r\n\r\n<Subheader>X. gflags</>\r\n\r\nCopyright (c) 2006, Google Inc.\r\nGflags software is licensed under the New BSD License.\r\n\r\n<Subheader>XI. GeometricTools</> \r\n\r\nGeometric Tools software is licensed under the Boost Software License Version 1.0 - August 17th, 2003.\r\n\r\n<Subheader>XII. Google Noto Fonts</>\r\n\r\nGoogle Noto Fonts is licensed under the SIL Open Font License, Version 1.1.\r\n\r\n<Subheader>XIII. ICU 51.2</>\r\n\r\nCopyright (c) 1995-2013 International Business Machines Corporation and others All rights reserved.\r\n\r\nPermission is hereby granted, free of charge, to any person obtaining a copy of this software and associated documentation files (the \"Software\"), to deal in the Software without restriction, including without limitation the rights to use, copy, modify, merge, publish, distribute, and/or sell copies of the Software, and to permit persons to whom the Software is furnished to do so, provided that the above copyright notice(s) and this permission notice appear in all copies of the Software and that both the above copyright notice(s) and this permission notice appear in supporting documentation.\r\n\r\nTHE SOFTWARE IS PROVIDED \"AS IS\", WITHOUT WARRANTY OF ANY KIND, EXPRESS OR IMPLIED, INCLUDING BUT NOT LIMITED TO THE WARRANTIES OF MERCHANTABILITY, FITNESS FOR A PARTICULAR PURPOSE AND NONINFRINGEMENT OF THIRD PARTY RIGHTS. IN NO EVENT SHALL THE COPYRIGHT HOLDER OR HOLDERS INCLUDED IN THIS NOTICE BE LIABLE FOR ANY CLAIM, OR ANY SPECIAL INDIRECT OR CONSEQUENTIAL DAMAGES, OR ANY DAMAGES WHATSOEVER RESULTING FROM LOSS OF USE, DATA OR PROFITS, WHETHER IN AN ACTION OF CONTRACT, NEGLIGENCE OR OTHER TORTIOUS ACTION, ARISING OUT OF OR IN CONNECTION WITH THE USE OR PERFORMANCE OF THIS SOFTWARE.\r\n\r\nExcept as contained in this notice, the name of a copyright holder shall not be used in advertising or otherwise to promote the sale, use or other dealings in this Software without prior written authorization of the copyright holder.\r\n\r\nICU 51.2 is licensed under the ICU License Version 1.8.1 and later.\r\n\r\n<Subheader>XIV. Jpeg Library</>\r\n\r\nFortnite is based in part on the work of the Independent JPEG Group.\r\n\r\n<Subheader>XV. Last Resort Font</>\r\n\r\nPLEASE READ THIS FONT SOFTWARE LICENSE AGREEMENT (\"LICENSE\") CAREFULLY BEFORE USING THE FONT SOFTWARE. BY USING THE FONT SOFTWARE, YOU ARE AGREEING TO BE BOUND BY THE TERMS OF THIS LICENSE. IF YOU DO NOT AGREE TO THE TERMS OF THIS LICENSE, DO NOT USE THE FONT SOFTWARE. IF THE FONT SOFTWARE WAS ACCESSED ELECTRONICALLY, CLICK \"DISAGREE/DECLINE\".\r\n\r\n1. General\r\nThe software and documentation accompanying this License whether on disk, in read only memory, on any other media or in any other form (collectively the \"Font Software\") are licensed, not sold, to you by the Unicode Consortium (\"Unicode\") for use only under the terms of this License, and Unicode reserves all rights not expressly granted to you. The License granted herein is granted under Unicode's and its licensors' intellectual property rights in the Font Software and does not include any other intellectual property rights. You own the media on which the Font Software is recorded but Unicode and/or Unicode's licensor(s) retain ownership of the Font Software itself. The terms of this License will govern any software upgrades provided by Unicode that replace and/or supplement the original Font Software product, unless such upgrade is accompanied by a separate license in which case the terms of that license will govern.\r\n\r\n2. Permitted License Uses and Restrictions\r\nIn consideration of your agreement to abide by the terms of this License, and subject to such terms, Unicode grants you a personal, non-exclusive license, to use, reproduce, and redistribute the Font Software. If you redistribute the Font Software, you must retain this entire license in all such redistributions. Except as and only to the extent expressly permitted in this License or by applicable law, you may not modify, decompile, reverse engineer, disassemble, or create derivative works of the Font Software or any part thereof. You may not sell, rent, lease, lend, or sublicense the Font Software. You may commercially distribute your materials that were created using the Font Software or that incorporate the Font Software, but the Font Software may not be commercially distributed on a standalone basis or as part of a commercially distributed package of fonts. THE FONT SOFTWARE IS NOT INTENDED FOR USE IN THE OPERATION OF NUCLEAR FACILITIES, AIRCRAFT NAVIGATION OR COMMUNICATION SYSTEMS, AIR TRAFFIC CONTROL SYSTEMS, LIFE SUPPORT MACHINES OR OTHER EQUIPMENT IN WHICH THE FAILURE OF THE FONT SOFTWARE COULD LEAD TO DEATH, PERSONAL INJURY, OR SEVERE PHYSICAL OR ENVIRONMENTAL DAMAGE. \r\n\r\n3. Termination\r\nThis License is effective until terminated. Your rights under this License will terminate automatically without notice from Unicode if you fail to comply with any term(s) of this License. Upon the termination of this License, you shall cease all use of the Font Software and destroy all copies, full or partial, of the Font Software.\r\n\r\n4. Disclaimer of Warranties\r\nYOU EXPRESSLY ACKNOWLEDGE AND AGREE THAT USE OF THE FONT SOFTWARE IS AT YOUR SOLE RISK AND THAT THE ENTIRE RISK AS TO SATISFACTORY QUALITY, PERFORMANCE, ACCURACY AND EFFORT IS WITH YOU. TO THE MAXIMUM EXTENT PERMITTED BY APPLICABLE LAW, THE FONT SOFTWARE IS PROVIDED \"AS IS\", WITH ALL FAULTS AND WITHOUT WARRANTY OF ANY KIND, AND UNICODE AND UNICODE'S LICENSORS (COLLECTIVELY REFERRED TO AS \"UNICODE\" FOR THE PURPOSES OF SECTIONS 4 AND 5) HEREBY DISCLAIM ALL WARRANTIES AND CONDITIONS WITH RESPECT TO THE FONT SOFTWARE, EITHER EXPRESS, IMPLIED OR STATUTORY, INCLUDING, BUT NOT LIMITED TO, THE IMPLIED WARRANTIES AND/OR CONDITIONS OF MERCHANTABILITY, OF SATISFACTORY QUALITY, OF FITNESS FOR A PARTICULAR PURPOSE, OF ACCURACY, OF QUIET ENJOYMENT, AND NON-INFRINGEMENT OF THIRD PARTY RIGHTS. UNICODE DOES NOT WARRANT AGAINST INTERFERENCE WITH YOUR ENJOYMENT OF THE FONT SOFTWARE, THAT THE FUNCTIONS CONTAINED IN THE FONT SOFTWARE WILL MEET YOUR REQUIREMENTS, THAT THE OPERATION OF THE FONT SOFTWARE WILL BE UNINTERRUPTED OR ERROR-FREE, OR THAT DEFECTS IN THE FONT SOFTWARE WILL BE CORRECTED. NO ORAL OR WRITTEN INFORMATION OR ADVICE GIVEN BY UNICODE OR A UNICODE-AUTHORIZED REPRESENTATIVE SHALL CREATE A WARRANTY. SHOULD THE FONT SOFTWARE PROVE DEFECTIVE, YOU ASSUME THE ENTIRE COST OF ALL NECESSARY SERVICING, REPAIR OR CORRECTION. SOME JURISDICTIONS DO NOT ALLOW THE EXCLUSION OF IMPLIED WARRANTIES OR LIMITATIONS ON APPLICABLE STATUTORY RIGHTS OF A CONSUMER, SO THE ABOVE EXCLUSION AND LIMITATIONS MAY NOT APPLY TO YOU.\r\n\r\n5. Limitation of Liabilities\r\nTO THE EXTENT NOT PROHIBITED BY LAW, IN NO EVENT SHALL UNICODE BE LIABLE FOR PERSONAL INJURY, OR ANY INCIDENTAL, SPECIAL, INDIRECT OR CONSEQUENTIAL DAMAGES WHATSOEVER, INCLUDING, WITHOUT LIMITATION, DAMAGES FOR LOSS OF PROFITS, LOSS OF DATA, BUSINESS INTERRUPTION OR ANY OTHER COMMERCIAL DAMAGES OR LOSSES, ARISING OUT OF OR RELATED TO YOUR USE OR INABILITY TO USE THE FONT SOFTWARE, HOWEVER CAUSED, REGARDLESS OF THE THEORY OF LIABILITY (CONTRACT, TORT OR OTHERWISE) AND EVEN IF UNICODE HAS BEEN ADVISED OF THE POSSIBILITY OF SUCH DAMAGES. SOME JURISDICTIONS DO NOT ALLOW THE LIMITATION OF LIABILITY FOR PERSONAL INJURY, OR OF INCIDENTAL OR CONSEQUENTIAL DAMAGES, SO THIS LIMITATION MAY NOT APPLY TO YOU. In no event shall Unicode's total liability to you for all damages (other than as may be required by applicable law in cases involving personal injury) exceed the amount of fifty dollars ($50.00). The foregoing limitations will apply even if the above stated remedy fails of its essential purpose.\r\n\r\n6. Export Control\r\nYou may not use or otherwise export or re-export the Font Software except as authorized by United States law and the laws of the jurisdiction in which the Font Software was obtained. In particular, but without limitation, the Font Software may not be exported or re-exported (a) into any U.S. embargoed countries or (b) to anyone on the U.S. Treasury Department's list of Specially Designated Nationals or the U.S. Department of Commerce's Denied Persons List or Entity List. By using the Font Software, you represent and warrant that you are not located in any such country or on any such list. You also agree that you will not use the Font Software for any purpose prohibited by United States law, including, without limitation, the development, design, manufacture or production of nuclear, missile, or chemical or biological weapons.\r\n\r\n7. Government End Users\r\nThe Font Software and related documentation are \"Commercial Items\", as that term is defined at 48 C.F.R. §2.101, consisting of \"Commercial Computer Software\" and \"Commercial Computer Software Documentation\", as such terms are used in 48 C.F.R. §12.212 or 48 C.F.R. §227.7202, as applicable. Consistent with 48 C.F.R. §12.212 or 48 C.F.R. §227.7202-1 through 227.7202-4, as applicable, the Commercial Computer Software and Commercial Computer Software Documentation are being licensed to U.S. Government end users (a) only as Commercial Items and (b) with only those rights as are granted to all other end users pursuant to the terms and conditions herein. Unpublished-rights reserved under the copyright laws of the United States.\r\n\r\n8. Controlling Laws and Severability\r\nThis License will be governed by and construed in accordance with the laws of the State of California, as applied to agreements entered into and to be performed entirely within California between California residents. This License shall not be governed by the United Nations Convention on Contracts for the International Sale of Goods, the application of which is expressly excluded. If for any reason a court of competent jurisdiction finds any provision, or portion thereof, to be unenforceable, the remainder of this License shall continue in full force and effect.\r\n\r\n9. Complete Agreement; Governing Language\r\nThis License constitutes the entire agreement between the parties with respect to the use of the Font Software licensed hereunder and supersedes all prior or contemporaneous understandings regarding such subject matter. No amendment to or modification of this License will be binding unless in writing and signed by Unicode. Any translation of this License is done for local requirements and in the event of a dispute between the English and any non-English version, the English version of this License shall govern.\r\n\r\n<Subheader>XVI. libcurl 7.47.1</>\r\n\r\nCopyright (c) 1996 - 2016, Daniel Stenberg, <daniel@haxx.se>, and many\r\ncontributors, see the THANKS file.\r\n\r\nLibcurl 7.47.1 software is licensed under the ISC License.\r\n\r\nExcept as contained in this notice, the name of a copyright holder shall not be used in advertising or otherwise to promote the sale, use or other dealings in this Software without prior written authorization of the copyright holder.\r\n\r\n<Subheader>XVII. Libwebsockets</>\r\n\r\nFortnite is based in part on the work of the libwebsockets project (<a id=\"url\" href=\"http://libwebsockets.org\" style=\"EULA.Hyperlink\">http://libwebsockets.org</>).\"\r\n\r\n<Subheader>XVIII. LibYUV</>\r\n       \r\nCopyright 2011 The LibYuv Project Authors. All rights reserved.\r\n\r\nLibYUV is licensed under the New BSD License.\r\n\r\n<Subheader>XIX. Mozilla source code – cacert.pem</>\r\n\r\nFortnite uses cacert.pem, which is subject to the terms of the Mozilla Public License v.2.0, available at (<a id=\"url\" href=\"http://mozilla.org/MPL/2.0/\" style=\"EULA.Hyperlink\">http://mozilla.org/MPL/2.0/</>). Source Code Form of cacert.pem is available at (<a id=\"url\" href=\"https://curl.haxx.se/docs/caextract.html\" style=\"EULA.Hyperlink\">https://curl.haxx.se/docs/caextract.html</>)\r\n \r\n<Subheader>XX. NanumGothic TTF</>\r\n\r\nCopyright (c) 2010, NHN Corporation (<a id=\"url\" href=\"http://www.nhncorp.com\" style=\"EULA.Hyperlink\">http://www.nhncorp.com</>), with Reserved Font Name Nanum, Naver Nanum, NanumGothic, Naver NanumGothic, NanumMyeongjo, Naver NanumMyeongjo, NanumBrush, Naver NanumBrush, NanumPen, Naver NanumPen, Naver NanumGothicEco, NanumGothicEco, Naver NanumMyeongjoEco, NanumMyeongjoEco, Naver NanumGothicLight, NanumGothicLight\r\n\r\nNanumGothic TTF is licensed under the SIL Open font License, Version 1.1.\r\n\r\n<Subheader>XXI. Noto Sans CJK-SC Font</>\r\n\r\nThis Font Software is licensed under the SIL Open Font License, Version 1.1.\r\n\r\n<Subheader>XXII. Ogg Vorbis Audio Compression</>\r\n\r\nCopyright (c) 2010, Xiph.Org Foundation\r\n\r\nOgg Vorbis Audio Compression software is dual-licensed to the public under the Creative Commons Attribution 3.0 License, or any later version published by Creative Commons, and the New BSD License.\r\n\r\n<Subheader>XXIII. Oodle SDK</>\r\n\r\nUses Oodle. Copyright (C) 2008-2017 by RAD Game Tools, Inc.\r\n\r\n<Subheader>XXIV. OpenEXR 1.7.1  + IlmBase 1.0.3</>\r\n\r\nCopyright (c) 2002-2011, Industrial Light && Magic, a division of Lucasfilm Entertainment Company Ltd. All rights reserved.\r\n\r\nOpenEXR 1.7.1  + IlmBase 1.0.3 software is licensed under the New BSD License.\r\n\r\n<Subheader>XXV. OpenSSL 1.0.2h</>\r\n\r\nCopyright (c) 1998-2016 The OpenSSL Project.  All rights reserved.\r\n \r\nRedistribution and use in source and binary forms, with or without modification, are permitted provided that the following conditions are met:\r\n \r\n1. Redistributions of source code must retain the above copyright notice, this list of conditions and the following disclaimer. \r\n\r\n2. Redistributions in binary form must reproduce the above copyright notice, this list of conditions and the following disclaimer in the documentation and/or other materials provided with the distribution.\r\n \r\n3. All advertising materials mentioning features or use of this software must display the following acknowledgment: \"This product includes software developed by the OpenSSL Project for use in the OpenSSL Toolkit. (<a id=\"url\" href=\"http://www.openssl.org/\" style=\"EULA.Hyperlink\">http://www.openssl.org/</>)\"\r\n\r\n4. The names \"OpenSSL Toolkit\" and \"OpenSSL Project\" must not be used to endorse or promote products derived from this software without prior written permission. For written permission, please contact openssl-core@openssl.org.\r\n\r\n5. Products derived from this software may not be called \"OpenSSL\" nor may \"OpenSSL\" appear in their names without prior written permission of the OpenSSL Project.\r\n \r\n6. Redistributions of any form whatsoever must retain the following acknowledgment: \"This product includes software developed by the OpenSSL Project for use in the OpenSSL Toolkit (http://www.openssl.org/)\"\r\n\r\nTHIS SOFTWARE IS PROVIDED BY THE OpenSSL PROJECT ``AS IS'' AND ANY EXPRESSED OR IMPLIED WARRANTIES, INCLUDING, BUT NOT LIMITED TO, THE IMPLIED WARRANTIES OF MERCHANTABILITY AND FITNESS FOR A PARTICULAR PURPOSE ARE DISCLAIMED.  IN NO EVENT SHALL THE OpenSSL PROJECT OR ITS CONTRIBUTORS BE LIABLE FOR ANY DIRECT, INDIRECT, INCIDENTAL, SPECIAL, EXEMPLARY, OR CONSEQUENTIAL DAMAGES (INCLUDING, BUT NOT LIMITED TO, PROCUREMENT OF SUBSTITUTE GOODS OR SERVICES; LOSS OF USE, DATA, OR PROFITS; OR BUSINESS INTERRUPTION) HOWEVER CAUSED AND ON ANY THEORY OF LIABILITY, WHETHER IN CONTRACT, STRICT LIABILITY, OR TORT (INCLUDING NEGLIGENCE OR OTHERWISE) ARISING IN ANY WAY OUT OF THE USE OF THIS SOFTWARE, EVEN IF ADVISED OF THE POSSIBILITY OF SUCH DAMAGE.\r\n\r\nThis product includes cryptographic software written by Eric Young (eay@cryptsoft.com).  This product includes software written by Tim Hudson (tjh@cryptsoft.com).\r\n\r\nOriginal SSLeay License\r\n\r\nCopyright (C) 1995-1998 Eric Young (eay@cryptsoft.com)\r\nAll rights reserved.\r\n\r\nThis package is an SSL implementation written by Eric Young (eay@cryptsoft.com).\r\nThe implementation was written so as to conform with Netscapes SSL.\r\n \r\nThis library is free for commercial and non-commercial use as long as the following conditions are aheared to.  The following conditions apply to all code found in this distribution, be it the RC4, RSA, lhash, DES, etc., code; not just the SSL code.  The SSL documentation included with this distribution is covered by the same copyright terms except that the holder is Tim Hudson (tjh@cryptsoft.com).\r\n\r\nCopyright remains Eric Young's, and as such any Copyright notices in the code are not to be removed. If this package is used in a product, Eric Young should be given attribution as the author of the parts of the library used. This can be in the form of a textual message at program startup or in documentation (online or textual) provided with the package.\r\n \r\nRedistribution and use in source and binary forms, with or without modification, are permitted provided that the following conditions are met:\r\n1. Redistributions of source code must retain the copyright notice, this list of conditions and the following disclaimer.\r\n2. Redistributions in binary form must reproduce the above copyright notice, this list of conditions and the following disclaimer in the documentation and/or other materials provided with the distribution.\r\n3. All advertising materials mentioning features or use of this software must display the following acknowledgement: \"This product includes cryptographic software written by Eric Young (eay@cryptsoft.com)\" The word 'cryptographic' can be left out if the rouines from the library being used are not cryptographic related :-).\r\n4. If you include any Windows specific code (or a derivative thereof) from  the apps directory (application code) you must include an acknowledgement: \"This product includes software written by Tim Hudson (tjh@cryptsoft.com)\"\r\n \r\nTHIS SOFTWARE IS PROVIDED BY ERIC YOUNG ``AS IS'' AND ANY EXPRESS OR IMPLIED WARRANTIES, INCLUDING, BUT NOT LIMITED TO, THE IMPLIED WARRANTIES OF MERCHANTABILITY AND FITNESS FOR A PARTICULAR PURPOSE ARE DISCLAIMED.  IN NO EVENT SHALL THE AUTHOR OR CONTRIBUTORS BE LIABLE FOR ANY DIRECT, INDIRECT, INCIDENTAL, SPECIAL, EXEMPLARY, OR CONSEQUENTIAL DAMAGES (INCLUDING, BUT NOT LIMITED TO, PROCUREMENT OF SUBSTITUTE GOODS OR SERVICES; LOSS OF USE, DATA, OR PROFITS; OR BUSINESS INTERRUPTION) HOWEVER CAUSED AND ON ANY THEORY OF LIABILITY, WHETHER IN CONTRACT, STRICT LIABILITY, OR TORT (INCLUDING NEGLIGENCE OR OTHERWISE) ARISING IN ANY WAY OUT OF THE USE OF THIS SOFTWARE, EVEN IF ADVISED OF THE POSSIBILITY OF SUCH DAMAGE.\r\n\r\nThe license and distribution terms for any publically available version or derivative of this code cannot be changed.  i.e. this code cannot simply be copied and put under another distribution license [including the GNU Public License.]\r\n\r\n<Subheader>XXVI. OpenSubDiv 3.0</>\r\n\r\nCopyright 2013 Pixar\r\n\r\nLicensed under the Apache License, Version 2.0 (the \"Apache License\") with the following modification; you may not use this file except in compliance with the Apache License and the following modification to it: Section 6. Trademarks. is deleted and replaced with:\r\n\r\n6. Trademarks. This License does not grant permission to use the trade names, trademarks, service marks, or product names of the Licensor and its affiliates, except as required to comply with Section 4(c) of the License and to reproduce the content of the NOTICE file.\r\n\r\nYou may obtain a copy of the Apache License at\r\n\r\n(<a id=\"url\" href=\"http://www.apache.org/licenses/LICENSE-2.0\" style=\"EULA.Hyperlink\">http://www.apache.org/licenses/LICENSE-2.0</>)\r\n\r\nUnless required by applicable law or agreed to in writing, software distributed under the Apache License with the above modification is distributed on an \"AS IS\" BASIS, WITHOUT WARRANTIES OR CONDITIONS OF ANY KIND, either express or implied. See the Apache License for the specific language governing permissions and limitations under the Apache License.\r\n\r\n<Subheader>XXVII. Opus 1.0.3</>\r\n\r\nOpus 1.0.3 software is licensed under the New BSD License.\r\n\r\n<Subheader>XXVIII. PLCrashReporter 1.2</>\r\n\r\nCopyright (c) 2008 - 2014 Plausible Labs Cooperative, Inc. All rights reserved. \r\nLicensed under the Apache License, Version 2.0. \r\n\r\n<Subheader>XXIX. PrecomputedAtmosphericScattering (Sample code, 08/16/2011)</>\r\n\r\nCopyright (c) 2008 INRIA.  All rights reserved.\r\n\r\n<Subheader>XXX. Protobuf</>\r\n\r\nThis license applies to all parts of Protocol Buffers except the following:\r\n\r\n- Atomicops support for generic gcc, located in src/google/protobuf/stubs/atomicops_internals_generic_gcc.h. \r\nThis file is copyrighted by Red Hat Inc.\r\n- Atomicops support for AIX/POWER, located in src/google/protobuf/stubs/atomicops_internals_aix.h. \r\nThis file is copyrighted by Bloomberg Finance LP.\r\n- Andorid.mk, which is copyrighted by The Android Open Source Project.\r\n\r\nCopyright 2014, Google Inc.  All rights reserved.\r\n\r\nRedistribution and use in source and binary forms, with or without modification, are permitted provided that the following conditions are met:\r\n\r\n- Redistributions of source code must retain the above copyright notice, this list of conditions and the following disclaimer.\r\n- Redistributions in binary form must reproduce the above copyright notice, this list of conditions and the following disclaimer in the documentation and/or other materials provided with the distribution.\r\n- Neither the name of Google Inc. nor the names of its contributors may be used to endorse or promote products derived from this software without specific prior written permission.\r\n\r\nTHIS SOFTWARE IS PROVIDED BY THE COPYRIGHT HOLDERS AND CONTRIBUTORS \"AS IS\" AND ANY EXPRESS OR IMPLIED WARRANTIES, INCLUDING, BUT NOT LIMITED TO, THE IMPLIED WARRANTIES OF MERCHANTABILITY AND FITNESS FOR A PARTICULAR PURPOSE ARE DISCLAIMED. IN NO EVENT SHALL THE COPYRIGHT OWNER OR CONTRIBUTORS BE LIABLE FOR ANY DIRECT, INDIRECT, INCIDENTAL, SPECIAL, EXEMPLARY, OR CONSEQUENTIAL DAMAGES (INCLUDING, BUT NOT LIMITED TO, PROCUREMENT OF SUBSTITUTE GOODS OR SERVICES; LOSS OF USE, DATA, OR PROFITS; OR BUSINESS INTERRUPTION) HOWEVER CAUSED AND ON ANY THEORY OF LIABILITY, WHETHER IN CONTRACT, STRICT LIABILITY, OR TORT (INCLUDING NEGLIGENCE OR OTHERWISE) ARISING IN ANY WAY OUT OF THE USE OF THIS SOFTWARE, EVEN IF ADVISED OF THE POSSIBILITY OF SUCH DAMAGE.\r\n\r\nCode generated by the Protocol Buffer compiler is owned by the owner of the input file used when generating it.  This code is not standalone and requires a support library to be linked with it.  This support library is itself covered by the above license.\r\n\r\n<Subheader>XXXI. Protobuf-c</>\r\n\r\nCopyright (c) 2008-2016, Dave Benson and the protobuf-c authors.\r\nAll rights reserved.\r\n\r\nRedistribution and use in source and binary forms, with or without modification, are permitted provided that the following conditions are met:\r\n\r\n- Redistributions of source code must retain the above copyright notice, this list of conditions and the following disclaimer.\r\n\r\n- Redistributions in binary form must reproduce the above copyright notice, this list of conditions and the following disclaimer in the documentation and/or other materials provided with the distribution.\r\n\r\nTHIS SOFTWARE IS PROVIDED BY THE COPYRIGHT HOLDERS AND CONTRIBUTORS \"AS IS\" AND ANY EXPRESS OR IMPLIED WARRANTIES, INCLUDING, BUT NOT LIMITED TO, THE IMPLIED WARRANTIES OF MERCHANTABILITY AND FITNESS FOR A PARTICULAR PURPOSE ARE DISCLAIMED. IN NO EVENT SHALL THE COPYRIGHT OWNER OR CONTRIBUTORS BE LIABLE FOR ANY DIRECT, INDIRECT, INCIDENTAL, SPECIAL, EXEMPLARY, OR CONSEQUENTIAL DAMAGES (INCLUDING, BUT NOT LIMITED TO, PROCUREMENT OF SUBSTITUTE GOODS OR SERVICES; LOSS OF USE, DATA, OR PROFITS; OR BUSINESS INTERRUPTION) HOWEVER CAUSED AND ON ANY THEORY OF LIABILITY, WHETHER IN CONTRACT, STRICT LIABILITY, OR TORT (INCLUDING NEGLIGENCE OR OTHERWISE) ARISING IN ANY WAY OUT OF THE USE OF THIS SOFTWARE, EVEN IF ADVISED OF THE POSSIBILITY OF SUCH DAMAGE.\r\n\r\n<Subheader>XXXII. Recast v1.4.2</>\r\n\r\nCopyright (c) 2009 Mikko Mononen memon@inside.org\r\n\r\nThis software is provided 'as-is', without any express or implied warranty.  In no event will the authors be held liable for any damages arising from the use of this software.\r\n\r\nPermission is granted to anyone to use this software for any purpose, including commercial applications, and to alter it and redistribute it freely, subject to the following restrictions:\r\n\r\n1. The origin of this software must not be misrepresented; you must not claim that you wrote the original software. If you use this software in a product, an acknowledgment in the product documentation would be appreciated but is not required.\r\n2. Altered source versions must be plainly marked as such, and must not be misrepresented as being the original software.\r\n3. This notice may not be removed or altered from any source distribution.\r\n\r\n<Subheader>XXXIII. Roboto Font</>\r\n\r\nApache License\r\nVersion 2.0, January 2004\r\n(<a id=\"url\" href=\"http://www.apache.org/licenses/\" style=\"EULA.Hyperlink\">http://www.apache.org/licenses/</>)\r\n\r\nTERMS AND CONDITIONS FOR USE, REPRODUCTION, AND DISTRIBUTION\r\n\r\nRoboto Font is licensed under the Apache License Version 2.0.\r\n\r\n<Subheader>XXXIV. Separable SSS</>\r\n\r\nCopyright (C) 2011 Jorge Jimenez (jorge@iryoku.com)\r\nCopyright (C) 2011 Diego Gutierrez (diegog@unizar.es)\r\nAll rights reserved.\r\n\r\nSeparable SSS software is licensed under the BSD License.\r\n\r\nThe views and conclusions contained in the software and documentation are those of the authors and should not be interpreted as representing official policies, either expressed or implied, of the copyright holders.\r\n\r\n<Subheader>XXXV. Unicode Common Locale Data Repository v23.1</>\r\n\r\nUNICODE, INC. LICENSE AGREEMENT - DATA FILES AND SOFTWARE\r\nUnicode Data Files include all data files under the directories <a id=\"url\" href=\"http://www.unicode.org/Public/\" style=\"EULA.Hyperlink\">http://www.unicode.org/Public/</>, <a id=\"url\" href=\"http://www.unicode.org/reports/\" style=\"EULA.Hyperlink\">http://www.unicode.org/reports/</>, and <a id=\"url\" href=\"http://www.unicode.org/cldr/data/\" style=\"EULA.Hyperlink\">http://www.unicode.org/cldr/data/</>. Unicode Data Files do not include PDF online code charts under the directory <a id=\"url\" href=\"http://www.unicode.org/Public/\" style=\"EULA.Hyperlink\">http://www.unicode.org/Public/</>. Software includes any source code published in the Unicode Standard or under the directories <a id=\"url\" href=\"http://www.unicode.org/Public/\" style=\"EULA.Hyperlink\">http://www.unicode.org/Public/</>, <a id=\"url\" href=\"http://www.unicode.org/reports/\" style=\"EULA.Hyperlink\">http://www.unicode.org/reports/</>, and <a id=\"url\" href=\"http://www.unicode.org/cldr/data/\" style=\"EULA.Hyperlink\">http://www.unicode.org/cldr/data/</>.\r\n\r\nNOTICE TO USER: Carefully read the following legal agreement. BY DOWNLOADING, INSTALLING, COPYING OR OTHERWISE USING UNICODE INC.'S DATA FILES (\"DATA FILES\"), AND/OR SOFTWARE (\"SOFTWARE\"), YOU UNEQUIVOCALLY ACCEPT, AND AGREE TO BE BOUND BY, ALL OF THE TERMS AND CONDITIONS OF THIS AGREEMENT. IF YOU DO NOT AGREE, DO NOT DOWNLOAD, INSTALL, COPY, DISTRIBUTE OR USE THE DATA FILES OR SOFTWARE.\r\n\r\nCOPYRIGHT AND PERMISSION NOTICE\r\nCopyright (c) 1991-2013 Unicode, Inc. All rights reserved. Distributed under the Terms of Use in http://www.unicode.org/copyright.html. Permission is hereby granted, free of charge, to any person obtaining a copy of the Unicode data files and any associated documentation (the \"Data Files\") or Unicode software and any associated documentation (the \"Software\") to deal in the Data Files or Software without restriction, including without limitation the rights to use, copy, modify, merge, publish, distribute, and/or sell copies of the Data Files or Software, and to permit persons to whom the Data Files or Software are furnished to do so, provided that (a) the above copyright notice(s) and this permission notice appear with all copies of the Data Files or Software, (b) both the above copyright notice(s) and this permission notice appear in associated documentation, and (c) there is clear notice in each modified Data File or in the Software as well as in the documentation associated with the Data File(s) or Software that the data or software has been modified.\r\n\r\nTHE DATA FILES AND SOFTWARE ARE PROVIDED \"AS IS\", WITHOUT WARRANTY OF ANY KIND, EXPRESS OR IMPLIED, INCLUDING BUT NOT LIMITED TO THE WARRANTIES OFMERCHANTABILITY, FITNESS FOR A PARTICULAR PURPOSE AND NONINFRINGEMENT OF THIRD PARTY RIGHTS. IN NO EVENT SHALL THE COPYRIGHT HOLDER OR HOLDERS INCLUDED IN THIS NOTICE BE LIABLE FOR ANY CLAIM, OR ANY SPECIAL INDIRECT OR CONSEQUENTIAL DAMAGES, OR ANY DAMAGES WHATSOEVER RESULTING FROM LOSS OF USE, DATA OR PROFITS, WHETHER IN AN ACTION OF CONTRACT, NEGLIGENCE OR OTHER TORTIOUS ACTION, ARISING OUT OF OR IN CONNECTION WITH THE USE OR PERFORMANCE OF THE DATA FILES OR SOFTWARE.\r\n\r\nExcept as contained in this notice, the name of a copyright holder shall not be used in advertising or otherwise to promote the sale, use or other dealings in these Data Files or Software without prior written authorization of the copyright holder.\r\n\r\n<Subheader>XXXVI. Vorbis Quality Improvements</>\r\n\r\naoTuV - Copyright (c) 2003-2008 Aoyumi\r\nlibvorbis - Copyright (c) 2002-2008 Xiph.org Foundation\r\n\r\nRedistribution and use in source and binary forms, with or without modification, are permitted provided that the following conditions are met:\r\n\r\n- Redistributions of source code must retain the above copyright notice, this list of conditions and the following disclaimer.\r\n\r\n- Redistributions in binary form must reproduce the above copyright notice, this list of conditions and the following disclaimer in the documentation and/or other materials provided with the distribution.\r\n\r\n- Neither the name of the Xiph.org Foundation nor the names of its contributors may be used to endorse or promote products derived from this software without specific prior written permission.\r\n\r\nTHIS SOFTWARE IS PROVIDED BY THE COPYRIGHT HOLDERS AND CONTRIBUTORS ``AS IS'' AND ANY EXPRESS OR IMPLIED WARRANTIES, INCLUDING, BUT NOT LIMITED TO, THE IMPLIED WARRANTIES OF MERCHANTABILITY AND FITNESS FOR A PARTICULAR PURPOSE ARE DISCLAIMED.  IN NO EVENT SHALL THE FOUNDATION OR CONTRIBUTORS BE LIABLE FOR ANY DIRECT, INDIRECT, INCIDENTAL, SPECIAL, EXEMPLARY, OR CONSEQUENTIAL DAMAGES (INCLUDING, BUT NOT LIMITED TO, PROCUREMENT OF SUBSTITUTE GOODS OR SERVICES; LOSS OF USE, DATA, OR PROFITS; OR BUSINESS INTERRUPTION) HOWEVER CAUSED AND ON ANY THEORY OF LIABILITY, WHETHER IN CONTRACT, STRICT LIABILITY, OR TORT (INCLUDING NEGLIGENCE OR OTHERWISE) ARISING IN ANY WAY OUT OF THE USE OF THIS SOFTWARE, EVEN IF ADVISED OF THE POSSIBILITY OF SUCH DAMAGE.\r\n\r\n<Subheader>XXXVII. WebRTC stable and trunk version</>\r\n\r\nCopyright (c) 2011, The WebRTC project authors. All rights reserved.\r\n\r\nWebRTC software is licensed under the New BSD License.\r\n\r\n<Subheader>XXXVIII. 3dRudder</>\r\n\r\nCopyright (c) 2014-2019, 3dRudder SA --- For terms of use: <a id=\"url\" href=\"https://3drudder-dev.com/docs/introduction/sdk_licensing_agreement/\" style=\"EULA.Hyperlink\">https://3drudder-dev.com/docs/introduction/sdk_licensing_agreement/</>\r\n\r\n<Subheader>XXXIX. ZPlane Elastique Pro V3</>\r\n\r\nZPlane Elastique Pro V3\r\n\"Fortnite uses élastique Pro V3 by Zplane as time-stretching engine.\""
    hyperlinkButtonStyle {
     clickedSlateSound {resourceObject=null}
     disabled {
      bIsDynamicallyLoaded=false
      drawAs=NoDrawType
      imageType=NoImage
      mirroring=NoMirror
      resourceName=None
      resourceObject=null
      tiling=NoTile
      imageSize=<0 0>
      margin=<ltrb 0 0 0 0>
      outlineSettings {
       bUseBrushTransparency=false
       roundingType=HalfHeightRadius
       width=0
       color {colorUseRule=UseColor_Specified, specifiedColor=<rgba 0 0 0 0>}
       cornerRadii=<0 0 0 0>
      }
      tintColor {colorUseRule=UseColor_Specified, specifiedColor=<rgba 1 1 1 1>}
      uVRegion {bIsValid=false, max=<0 0>, min=<0 0>}
     }
     disabledForeground {colorUseRule=UseColor_Foreground, specifiedColor=<rgba 1 0 1 1>}
     hovered {
      bIsDynamicallyLoaded=false
      drawAs=Image
      imageType=NoImage
      mirroring=NoMirror
      resourceName=None
      resourceObject=null
      tiling=NoTile
      imageSize=<32 32>
      margin=<ltrb 0 0 0 0>
      outlineSettings {
       bUseBrushTransparency=false
       roundingType=HalfHeightRadius
       width=0
       color {colorUseRule=UseColor_Specified, specifiedColor=<rgba 0 0 0 0>}
       cornerRadii=<0 0 0 0>
      }
      tintColor {colorUseRule=UseColor_Specified, specifiedColor=<rgba 1 1 1 0>}
      uVRegion {bIsValid=false, max=<0 0>, min=<0 0>}
     }
     hoveredForeground {colorUseRule=UseColor_Foreground, specifiedColor=<rgba 1 0 1 1>}
     hoveredSlateSound {resourceObject=null}
     normal {
      bIsDynamicallyLoaded=false
      drawAs=Image
      imageType=NoImage
      mirroring=NoMirror
      resourceName=None
      resourceObject=null
      tiling=NoTile
      imageSize=<32 32>
      margin=<ltrb 0 0 0 0>
      outlineSettings {
       bUseBrushTransparency=false
       roundingType=HalfHeightRadius
       width=0
       color {colorUseRule=UseColor_Specified, specifiedColor=<rgba 0 0 0 0>}
       cornerRadii=<0 0 0 0>
      }
      tintColor {colorUseRule=UseColor_Specified, specifiedColor=<rgba 1 1 1 0>}
      uVRegion {bIsValid=false, max=<0 0>, min=<0 0>}
     }
     normalForeground {colorUseRule=UseColor_Foreground, specifiedColor=<rgba 1 0 1 1>}
     normalPadding=<ltrb 0 0 0 0>
     pressed {
      bIsDynamicallyLoaded=false
      drawAs=Image
      imageType=NoImage
      mirroring=NoMirror
      resourceName=None
      resourceObject=null
      tiling=NoTile
      imageSize=<32 32>
      margin=<ltrb 0 0 0 0>
      outlineSettings {
       bUseBrushTransparency=false
       roundingType=HalfHeightRadius
       width=0
       color {colorUseRule=UseColor_Specified, specifiedColor=<rgba 0 0 0 0>}
       cornerRadii=<0 0 0 0>
      }
      tintColor {colorUseRule=UseColor_Specified, specifiedColor=<rgba 1 1 1 1>}
      uVRegion {bIsValid=false, max=<0 0>, min=<0 0>}
     }
     pressedForeground {colorUseRule=UseColor_Foreground, specifiedColor=<rgba 1 0 1 1>}
     pressedPadding=<ltrb 0 0 0 0>
     pressedSlateSound {resourceObject=null}
    }
   }
   slot {class=ScrollBoxSlot}
  }
 ]
}
