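# Fortnite legal information modal on FortLegalInfoWidget, with Terms of Service and Third Party Software tabs plus policy and close buttons.
# AddJapaneseEULATabs creates two EulaTab widgets keyed to Japanese legal texts and registers them as tabs.
# Activation plays a fade/scale intro and fires an analytics event; switching to the third-party tab calls its SetupScrollSTW.
# Construct and input changes apply compact-UI padding, except gamepad on mobile, which restores PC padding.
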WidgetBlueprint /Game/UI/Foundation/Legal/LegalInfo
animations [Intro {
  binding_count=1
  length_seconds=0.25
  object=&self:Intro
  sequence {bindings [CommonBorder_3 {
     id=#1
     possessed_class=CommonBorder
     tracks [
      {
       type=MovieSceneFloatTrack
       display_name=Render Opacity
       sections [{type=MovieSceneFloatSection, end_frame=5, start_frame=0, channels [None {type=MovieSceneScriptingFloatChannel, keys [[0,0],[2,0],[5,1]]}]}]
      }
      {
       type=MovieScene2DTransformTrack
       sections [{
         type=MovieScene2DTransformSection
         end_frame=5
         start_frame=0
         channels [
          Scale.X {type=MovieSceneScriptingFloatChannel, keys [[2,1.15],[4,1]]}
          Scale.Y {type=MovieSceneScriptingFloatChannel, keys [[2,1.15],[4,1]]}
         ]
        }]
      }
     ]
    }]}
 }]
graphs [
 EventGraph {nodes [
   n1 = "On Active Widget Changed (WidgetSwitcher_Legal)"!() -> (OutputDelegate, then, ActiveWidget:Widget Object Reference, ActiveWidgetIndex:int=0) {is=title K2Node_ComponentBoundEvent}
   > n2 = Cast To ThirdPartyTab!(execute, Object=@n1.ActiveWidget) -> (then, CastFailed, AsThird Party Tab:Third Party Tab Object Reference) {is=title K2Node_DynamicCast}
   > SetupScrollSTW(self:Third Party Tab Object Reference=@n2)
   n3 = "On Tab Button Creation (TabList_LegalOptions)"!() -> (OutputDelegate, then, TabId:Name=None, TabButton:Common Button Base Object Reference) {is=title K2Node_ComponentBoundEvent}
   > n4 = Cast To IconTextButton!(execute, Object=@n3.TabButton) -> (then, CastFailed, AsIcon Text Button:Icon Text Button Object Reference) {is=title K2Node_DynamicCast}
   > SetMouseKeyboardStyle!(
    execute
    self:Icon Text Button Object Reference=@n4
    ControllerInputStyle:Common Button Style Class Reference=/FortUILibrary/Foundation/Buttons/ButtonStyles/Tab/ButtonStyle-Tab-Secondary.ButtonStyle-Tab-Secondary_C
   ) -> (then)
   > SetContentAlignment!(execute, self:Icon Text Button Object Reference=@n4, ContentAlignment:EHorizontalAlignment Enum=HAlign_Center) -> (then)
   Event On Activated()
   > PlayAnimationForward(InAnimation=Get Intro())
   > Fire Event Sub View Opened!(execute, View:Widget Object Reference=Self-Reference(), AdditionalInfo:Flow Analytics Additional Info Structure, bIsOverlay:bool=false) -> (then)
   Event AddJapaneseEULATabs() {is=title K2Node_Event}
   > n5 = Sequence()
   > n6 = Create Widget!(
    execute:Exec
    Class:User Widget Class Reference=&/EulaTab.EulaTab_C
    OwningPlayer:Player Controller Object Reference=GetOwningPlayer()
    EULAKeyToShow:String=fn_jp_payment_services_act
   ) -> (then:Exec, ReturnValue:Eula Tab Object Reference)
   > AddChild(self=Get WidgetSwitcher_Legal() -> (WidgetSwitcher_Legal:Common Widget Switcher Legacy Object Reference), Content=@n6)
   > RegisterFortTab!(
    execute
    self:Fort Tab List Widget Base Legacy Object Reference=Get TabList_LegalOptions() -> (TabList_LegalOptions:Fort Tab List Widget Base Legacy Object Reference)
    TabNameId:Name=JAPaymentServices
    TabButtonType:Common Button Base Class Reference=/FortUILibrary/Foundation/Buttons/IconTextButton.IconTextButton_C
    LabelInfo_DisplayName:Text="INVTEXT(\"資金決済法に基づく表示\")"
    LabelInfo_IconBrush:Slate Brush Structure=(TintColor=(SpecifiedColor=(R=1,G=1,B=1,A=1),ColorUseRule=UseColor_Specified),DrawAs=NoDrawType,Tiling=NoTile,Mirroring=NoMirror,ImageType=NoImage,ImageSize=(X=0,Y=0),Margin=(Left=0,Top=0,Right=0,Bottom=0),ResourceObject=None,OutlineSettings=(CornerRadii=(X=0,Y=0,Z=0,W=0),Color=(SpecifiedColor=(R=0,G=0,B=0,A=0),ColorUseRule=UseColor_Specified),Width=0,RoundingType=HalfHeightRadius,bUseBrushTransparency=False),UVRegion=(Min=(X=0,Y=0),Max=(X=0,Y=0),bIsValid=False),bIsDynamicallyLoaded=False,ResourceName="")
    LabelInfo_IconResource:Soft Object Reference
    ContentWidget:Widget Object Reference=@n6
   ) -> (then, ReturnValue:bool)
   n7 = Create Widget(execute=@n5.then_1, Class=&EulaTab_C, OwningPlayer=GetOwningPlayer(), EULAKeyToShow=fn_jp_transaction_rules)
   > AddChild(self=Get WidgetSwitcher_Legal(), Content=@n7)
   > RegisterFortTab(
    self=Get TabList_LegalOptions()
    TabNameId=JATransactionRules
    TabButtonType=&IconTextButton_C
    LabelInfo_DisplayName="INVTEXT(\"特定商取引に基づく表示\")"
    LabelInfo_IconBrush=(TintColor=(SpecifiedColor=(R=1,G=1,B=1,A=1),ColorUseRule=UseColor_Specified),DrawAs=NoDrawType,Tiling=NoTile,Mirroring=NoMirror,ImageType=NoImage,ImageSize=(X=0,Y=0),Margin=(Left=0,Top=0,Right=0,Bottom=0),ResourceObject=None,OutlineSettings=(CornerRadii=(X=0,Y=0,Z=0,W=0),Color=(SpecifiedColor=(R=0,G=0,B=0,A=0),ColorUseRule=UseColor_Specified),Width=0,RoundingType=HalfHeightRadius,bUseBrushTransparency=False),UVRegion=(Min=(X=0,Y=0),Max=(X=0,Y=0),bIsValid=False),bIsDynamicallyLoaded=False,ResourceName="")
    ContentWidget=@n7
   )
   n8 = Event Construct()
   SetMobileStyle(execute=[@n8.then, @n10.else]) {is=title K2Node_CallFunction}
   n9 = Event OnInputModeChanged!() -> (OutputDelegate, then, bUsingGamepad:bool) {is=title K2Node_Event}
   > n10 = Branch(Condition=AND Boolean(A=@n9.bUsingGamepad, B=IsMobileGame()))
   > SetPadding!(
    execute
    self:Vertical Box Slot Object Reference=SlotAsVerticalBoxSlot(Widget=Get HorizontalBox_Buttons() -> (HorizontalBox_Buttons:Horizontal Box Object Reference))
    InPadding:Margin Structure=Get HboxButtons_PaddingPC() -> (HboxButtons_PaddingPC:Margin Structure)
   ) -> (then)
   > SetPadding(
    self=SlotAsVerticalBoxSlot(Widget=Get HorizontalBox_Header() -> (HorizontalBox_Header:Horizontal Box Object Reference))
    InPadding=Get HBoxHeaderPaddingPC() -> (HBoxHeaderPaddingPC:Margin Structure)
   )
  ]}
 SetMobileStyle {nodes [
   SetMobileStyle() {is=title K2Node_FunctionEntry}
   > Branch(Condition=IsUsingCompactUI())
   > SetPadding!(
    execute
    self=SlotAsVerticalBoxSlot(Widget=Get HorizontalBox_Buttons())
    InPadding_Left:float=32
    InPadding_Top:float=8
    InPadding_Right:float=0
    InPadding_Bottom:float=16
   ) -> (then)
   > SetPadding(self=SlotAsVerticalBoxSlot(Widget=Get HorizontalBox_Header()), InPadding_Left=36, InPadding_Top=0, InPadding_Right=0, InPadding_Bottom=0)
  ]}
]
variables [
 HboxButtons_PaddingPC {type=Margin, default=<ltrb 32 0 32 0>}
 HBoxHeaderPaddingPC {type=Margin, default=<ltrb 36 0 4 0>}
]
widgets {
 info {inheritedWidgetCount=6, parentClass=/Script/FortniteUI.FortLegalInfoWidget, rootWidgetClass=&Overlay, widgetCount=25}
 widgets [
  Overlay_0 {widgetClassPath=&Overlay, widget=Overlay_0}
  BackgroundBlurWidget_0 {
   widgetClassPath=&BackgroundBlur
   parent=Overlay_0
   widget=BackgroundBlurWidget_0
   properties {blurStrength=15}
   slot {class=OverlaySlot, properties {horizontalAlignment=HAlign_Fill, verticalAlignment=VAlign_Fill}}
  }
  Lightbox {
   widgetClassPath=/Game/UI/Foundation/Lightbox/Lightbox
   bIsVariable=true
   parent=Overlay_0
   widget=Lightbox
   properties {dimAmount=0.75, backgroundColor=<rgba 0 0.0006 0.0075 1>}
   slot {class=OverlaySlot, properties {horizontalAlignment=HAlign_Fill, verticalAlignment=VAlign_Fill}}
  }
  BorderTouch {
   widgetClassPath=&CommonBorder
   bIsVariable=true
   parent=Overlay_0
   widget=BorderTouch
   properties {visibility=Visible, padding=<ltrb 0 0 0 0>}
   slot {class=OverlaySlot, properties {horizontalAlignment=HAlign_Fill, verticalAlignment=VAlign_Fill}}
  }
  SB_Modal {
   widgetClassPath=&SizeBox
   parent=Overlay_0
   widget=SB_Modal
   properties {
    bOverride_HeightOverride=true
    bOverride_WidthOverride=true
    heightOverride=750
    minDesiredHeight=750
    minDesiredWidth=1200
    widthOverride=1200
   }
   slot {class=OverlaySlot, properties {horizontalAlignment=HAlign_Center, verticalAlignment=VAlign_Center, padding=<ltrb 96 0 96 0>}}
  }
  CommonBorder_3 {
   widgetClassPath=&CommonBorder
   parent=SB_Modal
   widget=CommonBorder_3
   properties {style=/FortUILibrary/Styles/Backgrounds/_BorderStyle_Modal._BorderStyle_Modal_C, padding=<ltrb 0 0 0 0>}
   slot {class=SizeBoxSlot}
  }
  Overlay_1 {widgetClassPath=&Overlay, parent=CommonBorder_3, widget=Overlay_1, slot {class=BorderSlot, properties {padding=<ltrb 0 0 0 0>}}}
  WBP_CaptureForPostBufferUpdate {
   widgetClassPath=&WBP_CaptureForPostBufferUpdate
   bIsVariable=true
   parent=Overlay_1
   slot {class=OverlaySlot, properties {horizontalAlignment=HAlign_Fill, verticalAlignment=VAlign_Fill}}
  }
  VerticalBox_354 {
   widgetClassPath=&VerticalBox
   parent=Overlay_1
   widget=VerticalBox_354
   slot {class=OverlaySlot, properties {horizontalAlignment=HAlign_Fill, verticalAlignment=VAlign_Fill}}
  }
  CommonBorder_1 {
   widgetClassPath=&CommonBorder
   parent=VerticalBox_354
   widget=CommonBorder_1
   properties {style=/FortUILibrary/Styles/Backgrounds/_BorderStyle_ModalHeader._BorderStyle_ModalHeader_C, padding=<ltrb 0 0 0 0>}
   slot {class=VerticalBoxSlot}
  }
  VerticalBox_0 {
   widgetClassPath=&VerticalBox
   parent=CommonBorder_1
   widget=VerticalBox_0
   slot {class=BorderSlot, properties {padding=<ltrb 0 0 0 0>}}
  }
  HorizontalBox_Header {
   widgetClassPath=&HorizontalBox
   bIsVariable=true
   parent=VerticalBox_0
   widget=HorizontalBox_Header
   slot {class=VerticalBoxSlot, properties {padding=<ltrb 36 0 4 0>, size {sizeRule=Fill, value=1}}}
  }
  Spacer_89 {
   widgetClassPath=&Spacer
   parent=HorizontalBox_Header
   widget=Spacer_89
   properties {size=<0 85>}
   slot {class=HorizontalBoxSlot, properties {horizontalAlignment=HAlign_Left, verticalAlignment=VAlign_Top}}
  }
  Title {
   widgetClassPath=&CommonTextBlock
   bIsVariable=true
   parent=HorizontalBox_Header
   widget=Title
   properties {
    justification=Center
    scrollStyle=&TextScrollStyle-Base_C
    style=/FortUILibrary/Styles/Typography/_TextStyle_SubH2._TextStyle_SubH2_C
    text=FORTNITE LEGAL INFORMATION
    font {
     bForceMonospaced=false
     bMaterialIsStencil=false
     fontMaterial=null
     fontObject=&Burbank
     letterSpacing=0
     monospacedWidth=1
     size=33
     skewAmount=0
     typefaceFontName=BurbankRegular-Black
     outlineSettings {
      bApplyOutlineToDropShadows=false
      bMiteredCorners=false
      bSeparateFillAlpha=false
      outlineMaterial=null
      outlineSize=0
      outlineColor=<rgba 0 0 0 1>
     }
    }
    shadowOffset=<0 0>
    strikeBrush {
     bIsDynamicallyLoaded=false
     drawAs=NoDrawType
     imageType=NoImage
     mirroring=NoMirror
     resourceName=None
     resourceObject=null
     tiling=NoTile
     imageSize=<32 32>
     margin=<ltrb 0 0 0 0>
     outlineSettings {
      bUseBrushTransparency=false
      roundingType=HalfHeightRadius
      width=0
      color {colorUseRule=UseColor_Specified, specifiedColor=<rgba 0 0 0 0>}
      cornerRadii=<0 0 0 0>
     }
     tintColor {colorUseRule=UseColor_Specified, specifiedColor=<rgba 1 1 1 1>}
     uVRegion {bIsValid=false, max=<0 0>, min=<0 0>}
    }
   }
   slot {class=HorizontalBoxSlot, properties {horizontalAlignment=HAlign_Left, verticalAlignment=VAlign_Center, size {sizeRule=Fill, value=1}}}
  }
  VisibilityBorder_Touch {
   widgetClassPath=&CommonHardwareVisibilityBorder
   parent=HorizontalBox_Header
   widget=VisibilityBorder_Touch
   properties {
    visibleType=Visible
    visibilityQuery {
     autoDescription=" ANY( Input.Touch )"
     tokenStreamVersion=0
     userDescription=""
     queryTokenStream [0,1,1,1,0]
     tagDictionary [{tagName=Input.Touch}]
    }
   }
   slot {class=HorizontalBoxSlot, properties {horizontalAlignment=HAlign_Right, padding=<ltrb 12 12 12 12>}}
  }
  Button_CloseTouch {
   widgetClassPath=/FortUILibrary/Foundation/Buttons/CloseButton
   bInherited=true
   bIsVariable=true
   parent=VisibilityBorder_Touch
   widget=Button_CloseTouch
   properties {
    onButtonBaseClicked=(&self:WidgetTree.Button_CloseTouch.HandleOnButtonClicked)
    onButtonBaseDoubleClicked=(&self:WidgetTree.Button_CloseTouch.HandleOnButtonDoubleClicked)
    onButtonBaseHovered=(&self:WidgetTree.Button_CloseTouch.HandleOnButtonHovered)
    onButtonBaseUnhovered=(&self:WidgetTree.Button_CloseTouch.HandleOnButtonUnhovered)
    onSelectedChangedBase=(&self:WidgetTree.Button_CloseTouch.HandleOnSelectedChanged)
   }
   slot {class=BorderSlot, properties {padding=<ltrb 0 0 0 0>}}
  }
  ScaleBox_196 {
   widgetClassPath=&ScaleBox
   parent=VerticalBox_0
   widget=ScaleBox_196
   slot {class=VerticalBoxSlot, properties {padding=<ltrb 36 -20 4 0>}}
  }
  TabList_LegalOptions {
   widgetClassPath=/Game/UI/Foundation/TabbedView/HorizontalTabList_Legacy
   bInherited=true
   bIsVariable=true
   parent=ScaleBox_196
   widget=TabList_LegalOptions
   properties {
    bAutoListenForInput=true
    buttonStyle=&ButtonStyle-Tab-Secondary_C
    onTabButtonCreation=(&self:WidgetTree.TabList_LegalOptions.HandleOnTabButtonCreated)
    onTabButtonRemoval=(&self:WidgetTree.TabList_LegalOptions.HandleOnTabButtonRemoved)
    useButtonStyle=true
    nextTabInputActionData {dataTable=&InputActionDataTable, rowName=SecondaryTabRight}
    preregisteredTabInfoArray [
     {
      bAllowedInZone=true
      bHidden=false
      tabButtonType=&IconTextButton_C
      tabContentType=&EulaTab_C
      tabNameId=EULA
      tabScene=null
      tabLabelInfo {
       displayName=TERMS OF SERVICE
       iconResource=null
       iconBrush {
        bIsDynamicallyLoaded=false
        drawAs=NoDrawType
        imageType=NoImage
        mirroring=NoMirror
        resourceName=None
        resourceObject=null
        tiling=NoTile
        imageSize=<0 0>
        margin=<ltrb 0 0 0 0>
        outlineSettings {
         bUseBrushTransparency=false
         roundingType=HalfHeightRadius
         width=0
         color {colorUseRule=UseColor_Specified, specifiedColor=<rgba 0 0 0 0>}
         cornerRadii=<0 0 0 1>
        }
        tintColor {colorUseRule=UseColor_Specified, specifiedColor=<rgba 1 1 1 1>}
        uVRegion {bIsValid=false, max=<0 0>, min=<0 0>}
       }
       priorityData {priority=Priority_3, whenContesting=QueueBehind}
      }
     }
     {
      bAllowedInZone=true
      bHidden=false
      tabButtonType=&IconTextButton_C
      tabContentType=&/ThirdPartyTab.ThirdPartyTab_C
      tabNameId=Third Party Software
      tabScene=null
      tabLabelInfo {
       displayName=Third Party Software
       iconResource=null
       iconBrush {
        bIsDynamicallyLoaded=false
        drawAs=NoDrawType
        imageType=NoImage
        mirroring=NoMirror
        resourceName=None
        resourceObject=null
        tiling=NoTile
        imageSize=<0 0>
        margin=<ltrb 0 0 0 0>
        outlineSettings {
         bUseBrushTransparency=false
         roundingType=HalfHeightRadius
         width=0
         color {colorUseRule=UseColor_Specified, specifiedColor=<rgba 0 0 0 0>}
         cornerRadii=<0 0 0 1>
        }
        tintColor {colorUseRule=UseColor_Specified, specifiedColor=<rgba 1 1 1 1>}
        uVRegion {bIsValid=false, max=<0 0>, min=<0 0>}
       }
       priorityData {priority=Priority_3, whenContesting=QueueBehind}
      }
     }
    ]
    previousTabInputActionData {dataTable=&InputActionDataTable, rowName=SecondaryTabLeft}
   }
   slot {class=ScaleBoxSlot, properties {horizontalAlignment=HAlign_Left, verticalAlignment=VAlign_Fill}}
  }
  WidgetSwitcher_Legal {
   widgetClassPath=&CommonWidgetSwitcherLegacy
   bInherited=true
   bIsVariable=true
   parent=VerticalBox_354
   widget=WidgetSwitcher_Legal
   properties {navigation=&self:WidgetTree.WidgetSwitcher_Legal.WidgetNavigation_0}
   slot {class=VerticalBoxSlot, properties {padding=<ltrb 32 16 32 0>, size {sizeRule=Fill, value=1}}}
  }
  HorizontalBox_Buttons {
   widgetClassPath=&HorizontalBox
   bIsVariable=true
   parent=VerticalBox_354
   widget=HorizontalBox_Buttons
   slot {class=VerticalBoxSlot, properties {padding=<ltrb 32 0 32 0>}}
  }
  Spacer {
   widgetClassPath=&Spacer
   parent=HorizontalBox_Buttons
   widget=Spacer
   properties {size=<0 80>}
   slot {class=HorizontalBoxSlot, properties {horizontalAlignment=HAlign_Left, verticalAlignment=VAlign_Top}}
  }
  Button_PrivacyPolicy {
   widgetClassPath=/FortUILibrary/UIKit/Buttons/WBP_UIKit_Button_Quiet
   bInherited=true
   bIsVariable=true
   parent=HorizontalBox_Buttons
   widget=Button_PrivacyPolicy
   properties {text=PRIVACY POLICY, triggeringInputAction {dataTable=&InputActionDataTable, rowName=GenericFaceButtonLeft}}
   slot {class=HorizontalBoxSlot, properties {verticalAlignment=VAlign_Center, padding=<ltrb 0 0 32 0>, size {sizeRule=Fill, value=0.9}}}
  }
  Button_FanContentPolicy {
   widgetClassPath=&WBP_UIKit_Button_Quiet
   bInherited=true
   bIsVariable=true
   parent=HorizontalBox_Buttons
   widget=Button_FanContentPolicy
   properties {text=FAN CONTENT POLICY, triggeringInputAction {dataTable=&InputActionDataTable, rowName=GenericFaceButtonTop}}
   slot {class=HorizontalBoxSlot, properties {verticalAlignment=VAlign_Center, padding=<ltrb 0 0 32 0>, size {sizeRule=Fill, value=1}}}
  }
  VisibilityBorder_NonTouch {
   widgetClassPath=&CommonHardwareVisibilityBorder
   parent=HorizontalBox_Buttons
   widget=VisibilityBorder_NonTouch
   properties {visibilityQuery {
     autoDescription=" NONE( Input.Touch )"
     tokenStreamVersion=0
     userDescription=""
     queryTokenStream [0,1,3,1,0]
     tagDictionary [{tagName=Input.Touch}]
    }}
   slot {class=HorizontalBoxSlot, properties {verticalAlignment=VAlign_Center, size {sizeRule=Fill, value=0.5}}}
  }
  Button_Close {
   widgetClassPath=&WBP_UIKit_Button_Quiet
   bInherited=true
   bIsVariable=true
   parent=VisibilityBorder_NonTouch
   widget=Button_Close
   properties {text=CLOSE, triggeringInputAction {dataTable=&InputActionDataTable, rowName=Close}}
   slot {class=BorderSlot, properties {padding=<ltrb 0 0 0 0>}}
  }
 ]
}
